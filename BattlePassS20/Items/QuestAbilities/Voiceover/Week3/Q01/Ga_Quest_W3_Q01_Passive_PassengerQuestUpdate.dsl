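; Passive quest ability for Battle Pass S20 Week 3 Q01 that shares a driver's objective progress with vehicle passengers.
; When the matching objective updates and the quest owner is driving, it clears ProcessedPassengers and loops over all passengers.
; Each passenger other than the owner gets a custom stat event through their quest manager.
; ProcessedPassengers records them, but its check does not block repeat sends.
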(event Ability|EventActivateAbility
  (|CallParentFunction))

(event Ability|EventFailedToActivatePassiveAbility
  (params (ActorInfo :type Gameplay%20Ability%20Actor%20Info%20Structure))
  (|CallParentFunction))

(event EventBP_OnObjectiveUpdated
  (params (QuestItem :type Fort%20Quest%20Item%20Object%20Reference) (ObjectiveBackendName :type Name :default "None") (AchivedCount :type Integer) (GoalCount :type Integer))
  (bind _returnvalue (Vehicle|GetVehicleInterface (Game|Player|GetPlayerPawn (Item|GetOwningController QuestItem))))
  (bind _returnvalue_2 (Item|GetOwningController QuestItem))
  (bind _returnvalue_3 (Game|Player|GetPlayerPawn _returnvalue_2))
  (if (== ObjectiveBackendName (Variables|Default|GetQuestBackendName))
    (bind _asfort_player_pawn_athena (Utilities|Casting|CastToFortPlayerPawnAthena (Vehicle|GetDriver _returnvalue)))
    (if (== _asfort_player_pawn_athena _returnvalue_3)
      (Utilities|Array|Clear (Variables|Default|GetProcessedPassengers))
      (for _array_element (Vehicle|GetAllPassengers _returnvalue)
        (bind _asfort_player_pawn_athena_2 (Utilities|Casting|CastToFortPlayerPawnAthena _array_element))
        (if (not (and (not (Utilities|Array|ContainsItem (Variables|Default|GetProcessedPassengers) _asfort_player_pawn_athena_2)) (== _asfort_player_pawn_athena_2 _returnvalue_3)))
          (Quest|SendComplexCustomStatEvent (Quests|GetQuestManager (Fortnite|GetFortPlayerControllerfromActor _asfort_player_pawn_athena_2) "Invalid") 0 (Variables|Default|GetT_Quest) "(GameplayTags=)" 1)
          (Utilities|Array|Add (Variables|Default|GetProcessedPassengers) _asfort_player_pawn_athena_2))))))
